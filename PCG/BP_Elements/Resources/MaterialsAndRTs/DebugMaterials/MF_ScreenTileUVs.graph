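# Material function that maps a rectangle of the viewport to tile UVs and a mask, stored with PCG debug materials.
# An inverse lerp of screen position between the min/max U and V inputs is saturated to give UVs.
# Max feeds the lerp's A, so UVs run from 0 at max to 1 at min.
# Floor and ceil of the UVs yield a branchless mask that is 1 only inside the rectangle.
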MaterialFunction /PCG/BP_Elements/Resources/MaterialsAndRTs/DebugMaterials/MF_ScreenTileUVs
graph {nodes [
  n1 = FunctionInput() {bUsePreviewValueAsDefault=true, inputName=maxU, inputType=FunctionInput_Scalar, sortPriority=2, previewValue=<0.25 0 0 1>}
  n2 = FunctionInput() {bUsePreviewValueAsDefault=true, inputName=maxV, inputType=FunctionInput_Scalar, sortPriority=4, previewValue=<0.5 0 0 1>}
  n3 = FunctionInput() {bUsePreviewValueAsDefault=true, inputName=minU, inputType=FunctionInput_Scalar, sortPriority=1}
  n4 = FunctionInput() {bUsePreviewValueAsDefault=true, inputName=minV, inputType=FunctionInput_Scalar, sortPriority=3}
  n5 = Saturate(None=InverseLinearInterpolate(A=AppendVector(A=@n1, B=@n2), B=AppendVector(A=@n3, B=@n4), Value=ScreenPosition().ViewportUV))
  FunctionOutput(None=@n5) {outputName=UVs}
  n6 = Multiply(A=OneMinus(None=Floor(None=@n5)), B=Ceil(None=@n5))
  FunctionOutput(None=Multiply(A=ComponentMask(None=@n6) {r=true}, B=ComponentMask(None=@n6) {g=true})) {outputName=Mask}
 ]}
