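# Post-process debug material that overlays eight texture parameters, Slot0 to Slot7, as a row of tiles across the top quarter of the screen.
# MF_ScreenTileUVs supplies each tile's UVs and mask, and the masked samples are summed.
# The sum gets linear-to-sRGB encoding, though the low-range multiply has no color input wired.
# The combined AllMasks reroute cuts the tiles out of the scene color before both are added.
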Material /PCG/BP_Elements/Resources/MaterialsAndRTs/DebugMaterials/PP_DebugMat
graph {
 properties {materialDomain=MD_PostProcess, pixelDepthOffsetMode=PDOM_Legacy, shadingModel=MSM_Unlit, usedShadingModels=Unlit}
 nodes [
  n1 = Constant() {r=0.125}
  n2 = Constant()
  n3 = Constant() {r=0.25}
  n4 = MaterialFunctionCall!(minU=Constant(), maxU=@n1, minV=@n2, maxV=@n3) {materialFunction=&/MF_ScreenTileUVs}
  n5 = TextureObjectParameter() {parameterName=Slot0, texture=&/debugFrame}
  n6 = Constant() {r=0.25}
  n7 = MaterialFunctionCall(minU=@n1, maxU=@n6, minV=@n2, maxV=@n3) {materialFunction=&MF_ScreenTileUVs}
  n8 = TextureObjectParameter() {parameterName=Slot1, texture=&debugFrame}
  n9 = Add(A=Multiply(A=@n4.Mask, B=TextureSample(UVs=@n4.UVs, Tex=@n5).RGB), B=Multiply(A=@n7.Mask, B=TextureSample(UVs=@n7.UVs, Tex=@n8).RGB))
  n10 = Constant() {r=0.375}
  n11 = MaterialFunctionCall(minU=@n6, maxU=@n10, minV=@n2, maxV=@n3) {materialFunction=&MF_ScreenTileUVs}
  n12 = Multiply(
   A=@n11.Mask
   B=TextureSample(
    UVs=@n11.UVs
    Tex=TextureObjectParameter() {parameterName=Slot2, texture=&debugFrame}
   ).RGB
  )
  n13 = Constant() {r=0.5}
  n14 = MaterialFunctionCall(minU=@n10, maxU=@n13, minV=@n2, maxV=@n3) {materialFunction=&MF_ScreenTileUVs}
  n15 = TextureSample(
   UVs=@n14.UVs
   Tex=TextureObjectParameter() {parameterName=Slot3, texture=&debugFrame}
  )
  n16 = Constant() {r=0.625}
  n17 = MaterialFunctionCall(minU=@n13, maxU=@n16, minV=@n2, maxV=@n3) {materialFunction=&MF_ScreenTileUVs}
  n18 = TextureObjectParameter() {parameterName=Slot4, texture=&debugFrame}
  n19 = Add(A=Add(A=Add(A=@n9, B=@n12), B=Multiply(A=@n14.Mask, B=@n15.RGB)), B=Multiply(A=@n17.Mask, B=TextureSample(UVs=@n17.UVs, Tex=@n18).RGB))
  n20 = Constant() {r=0.75}
  n21 = MaterialFunctionCall(minU=@n16, maxU=@n20, minV=@n2, maxV=@n3) {materialFunction=&MF_ScreenTileUVs}
  n22 = Multiply(
   A=@n21.Mask
   B=TextureSample(
    UVs=@n21.UVs
    Tex=TextureObjectParameter() {parameterName=Slot5, texture=&debugFrame}
   ).RGB
  )
  n23 = Constant() {r=0.875}
  n24 = MaterialFunctionCall(minU=@n20, maxU=@n23, minV=@n2, maxV=@n3) {materialFunction=&MF_ScreenTileUVs}
  n25 = TextureSample(
   UVs=@n24.UVs
   Tex=TextureObjectParameter() {parameterName=Slot6, texture=&debugFrame}
  )
  n26 = MaterialFunctionCall(minU=@n23, maxU=Constant() {r=1}, minV=@n2, maxV=@n3) {materialFunction=&MF_ScreenTileUVs}
  n27 = TextureObjectParameter() {parameterName=Slot7, texture=&debugFrame}
  n28 = Add(A=Add(A=Add(A=@n19, B=@n22), B=Multiply(A=@n24.Mask, B=@n25.RGB)), B=Multiply(A=@n26.Mask, B=TextureSample(UVs=@n26.UVs, Tex=@n27).RGB))
  n29 = Multiply(A=Power(Base=@n28, Exp=Constant() {desc=1/2.4, r=0.4167}), B=Constant() {r=1.055})
  n30 = Constant() {r=-0.055}
  n31 = Multiply(B=Constant() {r=12.92})
  n32 = SceneTexture() {sceneTextureId=PPI_PostProcessInput0}
  n33 = NamedRerouteUsage()
  Add(
   A=If(A=@n28, B=Constant() {desc=0.04045/12.92, r=0.0031}, "A > B"=Add(A=@n29, B=@n30), "A == B"=@n31, "A < B"=@n31)
   B=Multiply(A=ComponentMask(None=@n32.Color) {b=true, g=true, r=true}, B=OneMinus(None=@n33))
  )
  n34 = Max(A=Max(A=@n4.Mask, B=@n7.Mask), B=@n11.Mask)
  n35 = Max(A=Max(A=Max(A=@n34, B=@n14.Mask), B=@n17.Mask), B=@n21.Mask)
  NamedRerouteDeclaration(None=Max(A=Max(A=@n35, B=@n24.Mask), B=@n26.Mask)) {name=AllMasks, nodeColor=<rgba 0.5703 1 0 1>}
 ]
}
